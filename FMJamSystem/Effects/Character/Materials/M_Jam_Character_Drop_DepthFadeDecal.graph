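# Translucent unlit particle material for a Jam character "drop" effect, apparently a decal-like overlay that shows near scene geometry through depth fades.
# An inverted-V gradient distorted by blurred noise, minus the Anim dynamic parameter, drives an animated dissolve with a brightened edge on the particle color.
# Opacity math combines that dissolve, particle alpha, paired DepthFades and a sphere-masked Fresnel from flattened scene normals, with forward-shading and low-feature-level fallbacks.
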Material /FMJamSystem/Effects/Character/Materials/M_Jam_Character_Drop_DepthFadeDecal
graph {
 properties {
  bUsedWithMeshParticles=true
  bUsedWithNiagaraMeshParticles=true
  bUsedWithNiagaraSprites=true
  bUsedWithParticleSprites=true
  blendMode=BLEND_Translucent
  pixelDepthOffsetMode=PDOM_Legacy
  shadingModel=MSM_Unlit
  usedShadingModels=Unlit
 }
 nodes [
  n1 = ParticleColor()
  n2 = ScalarParameter() {defaultValue=10, parameterName=Param_3}
  n3 = Multiply(
   A=TextureCoordinate() {uTiling=5}
   B=ComponentMask(None=VectorParameter() {parameterName=Param_2, defaultValue=<rgba 1 0.25 0 1>}.RGB) {g=true, r=true}
  )
  n4 = Power(Base=OneMinus(None=TextureSample(UVs=@n3) {samplerType=SAMPLERTYPE_Grayscale, texture=&LowResBlurredNoise}.R))
  n5 = ComponentMask(None=OneMinus(None=TextureCoordinate())) {g=true}
  n6 = Subtract(A=Add(A=@n5, B=Multiply(A=@n5, B=@n4)), B=DynamicParameter() {paramNames [Anim,Param2,Param3,Param4]}.Anim)
  n7 = Divide(A=@n6, B=ScalarParameter() {defaultValue=0.1, parameterName=Param_2})
  n8 = Power(Base=Saturate(None=OneMinus(None=@n7)))
  Add(A=@n1.RGB, B=Multiply(A=@n1.RGB, B=Multiply(A=@n2, B=@n8)))
  Multiply(
   A=MaterialFunctionCall() {materialFunction=&VectorToRadialValue}.Radial Coordinates
   B=ComponentMask(None=VectorParameter() {parameterName=Param_1, defaultValue=<rgba 10 0.25 0 1>}.RGB) {g=true, r=true}
  )
  n9 = Divide(A=@n6, B=ScalarParameter() {defaultValue=0.01}) {constB=0.1}
  n10 = SceneTexture() {sceneTextureId=PPI_WorldNormal}
  n11 = Fresnel(
   ExponentIn=ScalarParameter() {defaultValue=2, parameterName=FresnelExponent}
   Normal=MaterialFunctionCall(Normal=ComponentMask(None=@n10.Color) {b=true, g=true, r=true}, Flatness=ScalarParameter() {defaultValue=0.15, parameterName=FlattenNormal}) {materialFunction=&FlattenNormal}.Result
  ) {baseReflectFraction=0, exponent=1}
  n12 = ShadingPathSwitch(Default=@n11, Deferred=@n11, Forward=VectorParameter() {parameterName=Forward, defaultValue=<rgba 0 0 1 1>}.RGB, Mobile=@n11)
  n13 = Constant() {r=0.05}
  n14 = TextureCoordinate()
  n15 = Constant() {r=0.5}
  n16 = ScalarParameter() {defaultValue=0.5, parameterName=SphereMaskRadius}
  n17 = ScalarParameter() {parameterName=SphereMaskHardness}
  n18 = Add(B=Multiply(
    A=FeatureLevelSwitch(Default=@n12, ES2=@n13, ES3_1=@n13, SM4_REMOVED=@n12, SM5=@n12, SM6=@n12)
    B=SphereMask(A=@n14, B=@n15, Radius=@n16, Hardness=@n17) {hardnessPercent=0}
   ))
  n19 = ScalarParameter() {defaultValue=100, parameterName=VisibleDepthDistance}
  n20 = DepthFade(
   Opacity=OneMinus(None=DepthFade(FadeDistance=@n19) {fadeDistanceDefault=80})
   FadeDistance=ScalarParameter() {defaultValue=15, parameterName=DepthFade}
  ) {fadeDistanceDefault=15}
  Multiply(A=ParticleColor().A, B=Multiply(A=Saturate(None=@n9), B=Multiply(A=@n18, B=@n20)))
  n21 = Time()
  NamedRerouteDeclaration(None=OneMinus(None=Frac(None=@n21))) {name=TimeTest, nodeColor=<rgba 1 0.75 0 1>}
  NamedRerouteUsage()
  SphereMask(
   A=TextureCoordinate()
   B=Constant() {r=0.5}
   Radius=ScalarParameter() {defaultValue=0.5, parameterName=SphereMaskRadius}
   Hardness=ScalarParameter() {parameterName=SphereMaskHardness}
  ) {hardnessPercent=0}
  VectorParameter() {defaultValue=<rgba 0 1 0.0885 1>}
 ]
}
