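# Unlit translucent material for a Jam character ring effect, drawn on Niagara mesh particles built from thin spline geometry.
# Emissive is particle color times an intensity, passed through the shared FX emissive adjustments.
# Opacity combines a sharpened procedural band, a smoothstepped sine mask along V, particle alpha, depth fade and inverted Fresnel.
# SplineThicken widens the strip by a mask-scaled width and supplies the Fresnel normal.
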Material /FMJamSystem/Effects/Character/Materials/M_Jam_Character_Ring_Spline
graph {
 properties {
  bCastRayTracedShadows=false
  bUsedWithNiagaraMeshParticles=true
  blendMode=BLEND_Translucent
  pixelDepthOffsetMode=PDOM_Legacy
  shadingModel=MSM_Unlit
  usedShadingModels=Unlit
 }
 nodes [
  n1 = ParticleColor()
  MaterialFunctionCall(Emissive=Multiply(A=@n1.RGB, B=ScalarParameter() {defaultValue=3, parameterName=ColorIntensity})) {materialFunction=&MF_FX_EmissiveAdjustments}
  n2 = MaterialFunctionCall!(
   Width=ScalarParameter() {defaultValue=1, desc=0.2, parameterName=BandWidth}
   Sharpness=ScalarParameter() {desc=1, parameterName=BandSharpness}
   Offset
   Direction Switch=StaticBool() {value=true}
   Compare
   Input Coordinates
  ) {materialFunction=/Engine/Functions/Engine_MaterialFunctions03/Procedurals/GeneratedBand}
  n3 = ScalarParameter() {defaultValue=10, parameterName=BandPower}
  n4 = Constant() {r=-0.2}
  n5 = Constant() {r=1}
  n6 = TextureCoordinate()
  n7 = Sine(None=Multiply(A=ComponentMask(None=@n6) {g=true}) {constB=0.5})
  n8 = Multiply(
   A=Multiply(A=Power(Base=@n2.Result, Exp=@n3), B=ScalarParameter() {defaultValue=0.5, parameterName=BandMult})
   B=NamedRerouteDeclaration(None=SmoothStep(Min=@n4, Max=@n5, Value=@n7)) {name=Camera Facing Mask, nodeColor=<rgba 0.2891 1 0 1>}
  )
  n9 = Constant() {r=2}
  n10 = Add(A=Multiply(A=ScalarParameter() {defaultValue=4, parameterName=SplineWidth}, B=NamedRerouteUsage()))
  n11 = MaterialFunctionCall(WidthBase=@n10, WidthTip=@n10) {materialFunction=&SplineThicken}
  Multiply(A=DepthFade(Opacity=Multiply(A=@n1.A, B=@n8)), B=OneMinus(None=Fresnel(ExponentIn=@n9, Normal=@n11.Normal)))
 ]
}
