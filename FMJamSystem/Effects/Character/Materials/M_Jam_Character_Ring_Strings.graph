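# Translucent unlit Niagara mesh particle material for audio-reactive strings in the Jam character ring effect.
# SplineThicken sets strand width, alongside named-reroute wave logic that builds a Z-axis world position offset from panning sines, four dynamic-parameter audio bands and a lerp to audio-free motion.
# Squared particle color drives emissive, while particle alpha scales two soft GeneratedBand profiles for opacity.
# A further branch builds band masks whose width follows switchable real or fake audio waves.
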Material /FMJamSystem/Effects/Character/Materials/M_Jam_Character_Ring_Strings
graph {
 properties {
  bCastRayTracedShadows=false
  bOutputTranslucentVelocity=true
  bUsedWithNiagaraMeshParticles=true
  blendMode=BLEND_Translucent
  opacityMaskClipValue=1
  pixelDepthOffsetMode=PDOM_Legacy
  shadingModel=MSM_Unlit
  usedShadingModels=Unlit
 }
 nodes [
  n1 = ScalarParameter() {defaultValue=7, parameterName=Spline_Width}
  Add(
   A=MaterialFunctionCall(WidthBase=@n1, WidthTip=@n1) {materialFunction=&SplineThicken}.WorldPositionOffset
   B=NamedRerouteUsage()
  )
  n2 = Panner(Coordinate=TextureCoordinate() {uTiling=4, vTiling=4}) {speedX=0.25, speedY=0.25}
  n3 = ConstantBiasScale(None=ComponentMask(None=Sine(None=@n2)) {g=true})
  n4 = Panner(Coordinate=TextureCoordinate() {uTiling=7, vTiling=7}) {speedX=0.25, speedY=0.25}
  n5 = ConstantBiasScale(None=ComponentMask(None=Sine(None=@n4)) {g=true})
  Add(A=Add(A=Multiply(A=@n3) {constB=0.2}, B=Multiply(A=@n5) {constB=0.2})) {constB=0.05}
  n6 = TextureCoordinate() {uTiling=3, vTiling=3}
  n7 = ComponentMask(None=Sine(None=Panner(Coordinate=@n6) {speedY=0.25})) {g=true}
  ConstantBiasScale(None=@n7)
  n8 = TextureCoordinate() {uTiling=4, vTiling=4}
  n9 = ComponentMask(None=Sine(None=Panner(Coordinate=@n8) {speedY=-0.25})) {g=true}
  ConstantBiasScale(None=@n9)
  n10 = TextureCoordinate() {uTiling=5, vTiling=5}
  n11 = ComponentMask(None=Sine(None=Panner(Coordinate=@n10) {speedY=-0.4})) {g=true}
  ConstantBiasScale(None=@n11)
  n12 = TextureCoordinate() {uTiling=8, vTiling=8}
  n13 = ComponentMask(None=Cosine(None=Panner(Coordinate=@n12) {speedY=0.4})) {g=true}
  ConstantBiasScale(None=@n13)
  n14 = ParticleColor()
  n15 = Multiply(A=@n14.RGB, B=@n14.RGB)
  n16 = ScalarParameter() {defaultValue=0.1, parameterName=ColorAdd}
  MaterialFunctionCall(Emissive=Multiply(A=Add(A=@n15, B=@n16) {constB=0.25}, B=ScalarParameter() {defaultValue=1, parameterName=ColorIntensity})) {materialFunction=&MF_FX_EmissiveAdjustments}
  n17 = ScalarParameter() {defaultValue=0.1, parameterName=Band_1_Width}
  n18 = ScalarParameter() {defaultValue=0.1, parameterName=Band_1_Sharpness}
  n19 = StaticBool() {value=true}
  n20 = Multiply(
   A=Power(
    Base=MaterialFunctionCall!(Width=@n17, Sharpness=@n18, Offset, Direction Switch=@n19, Compare, Input Coordinates) {materialFunction=/Engine/Functions/Engine_MaterialFunctions03/Procedurals/GeneratedBand}.Result
    Exp=ScalarParameter() {defaultValue=2, parameterName=Band_1_Pow}
   )
   B=ScalarParameter() {defaultValue=0.5, parameterName=Band_1_Mult}
  )
  n21 = ScalarParameter() {defaultValue=0.5, parameterName=Band_2_Width}
  n22 = ScalarParameter() {parameterName=Band_2_Sharpness}
  n23 = StaticBool() {value=true}
  n24 = Multiply(
   A=Power(
    Base=MaterialFunctionCall(Width=@n21, Sharpness=@n22, Direction Switch=@n23) {materialFunction=&GeneratedBand}.Result
    Exp=ScalarParameter() {defaultValue=2, parameterName=Band_2_Pow}
   )
   B=ScalarParameter() {defaultValue=0.5, parameterName=Band_2_Mult}
  )
  Multiply(A=Multiply(A=@n14.A, B=Add(A=@n20, B=@n24)), B=NamedRerouteUsage())
  n25 = DynamicParameter() {desc=Audio values, parameterIndex=1, defaultValue=<rgba 0 0 0 0>, paramNames [Band1,Band2,Band3,Band4]}
  NamedRerouteDeclaration(None=@n25.Band2) {name=Band2, nodeColor=<rgba 1 0 0.6328 1>}
  NamedRerouteDeclaration(None=@n25.Band3) {name=Band3, nodeColor=<rgba 0.2656 0 1 1>}
  NamedRerouteDeclaration(None=@n25.Band4) {name=Band4, nodeColor=<rgba 1 0 0.9609 1>}
  NamedRerouteDeclaration(None=@n7) {name=Sine1, nodeColor=<rgba 1 0 0.9609 1>}
  n26 = Add(A=Multiply(A=NamedRerouteUsage(), B=NamedRerouteUsage()), B=Multiply(A=NamedRerouteUsage(), B=NamedRerouteUsage()))
  n27 = Multiply(A=NamedRerouteUsage(), B=NamedRerouteUsage())
  n28 = NamedRerouteUsage()
  n29 = NamedRerouteUsage()
  n30 = Multiply(A=Add(A=Add(A=@n26, B=@n27), B=Multiply(A=@n28, B=@n29)), B=ScalarParameter() {defaultValue=75, parameterName=AudioIntensity}) {constB=75}
  n31 = DynamicParameter() {defaultValue=<rgba 1 0 1 1>, paramNames [LoopAmplitudeMultipy, Param2, Param3, Param4]}
  n32 = NamedRerouteUsage()
  n33 = NamedRerouteUsage()
  n34 = Add(A=Add(A=Add(A=@n32, B=@n33), B=NamedRerouteUsage()), B=NamedRerouteUsage())
  NamedRerouteDeclaration(None=LinearInterpolate(A=Multiply(A=@n30, B=@n31.LoopAmplitudeMultipy), B=Multiply(A=@n34) {constB=2}, Alpha=@n31.Param2)) {name=4BandSineWaves, nodeColor=<rgba 0 0 0 1>}
  NamedRerouteDeclaration(None=@n9) {name=Sine2, nodeColor=<rgba 0 0.7422 1 1>}
  NamedRerouteDeclaration(None=@n13) {name=Sine3, nodeColor=<rgba 1 0 0.9375 1>}
  NamedRerouteDeclaration(None=@n11) {name=Sine4, nodeColor=<rgba 0 0.0625 1 1>}
  n35 = DynamicParameter() {desc=Audio values, parameterIndex=1, defaultValue=<rgba 0 0 0 0>, paramNames [Band1,Band2,Band3,Band4]}
  NamedRerouteDeclaration(None=@n35.Band2) {name=Band2_1, nodeColor=<rgba 1 0 0.6328 1>}
  NamedRerouteDeclaration(None=@n35.Band3) {name=Band3_1, nodeColor=<rgba 0.2656 0 1 1>}
  NamedRerouteDeclaration(None=@n35.Band4) {name=Band4_1, nodeColor=<rgba 1 0 0.9609 1>}
  n36 = Multiply(A=ComponentMask(None=TextureCoordinate()) {g=true}) {constB=0.5}
  NamedRerouteDeclaration(None=SmoothStep(Min=Constant() {r=-0.1}, Max=Constant() {r=1}, Value=Sine(None=@n36))) {name=Camera Facing Mask, nodeColor=<rgba 0.2891 1 0 1>}
  NamedRerouteDeclaration(None=@n25.Band1) {name=Band1, nodeColor=<rgba 1 0.7031 0 1>}
  NamedRerouteDeclaration(None=Multiply(A=NamedRerouteUsage(), B=Constant3Vector() {desc=WorldPositionDirection, constant=<rgba 0 0 1 0>})) {name=Final WPO, nodeColor=<rgba 0 0 0 1>}
  NamedRerouteDeclaration(None=@n35.Band1) {name=Band1_1, nodeColor=<rgba 1 0.7031 0 1>}
  n37 = Add(A=Multiply(A=NamedRerouteUsage(), B=NamedRerouteUsage()), B=Multiply(A=NamedRerouteUsage(), B=NamedRerouteUsage()))
  n38 = Multiply(A=NamedRerouteUsage(), B=NamedRerouteUsage())
  n39 = NamedRerouteUsage()
  n40 = NamedRerouteUsage()
  n41 = Multiply(A=Add(A=Add(A=@n37, B=@n38), B=Multiply(A=@n39, B=@n40)), B=ScalarParameter() {defaultValue=3, parameterName=Param_12}) {desc=SinewaveMultiply 500}
  n42 = DynamicParameter() {defaultValue=<rgba 1 0 1 1>, paramNames [LoopAmplitudeMultipy, Param2, Param3, Param4]}
  n43 = Sine(None=Panner(Coordinate=TextureCoordinate() {uTiling=3, vTiling=3}) {speedX=0.25, speedY=0.25})
  n44 = NamedRerouteDeclaration(None=ConstantBiasScale(None=ComponentMask(None=@n43) {g=true})) {name=Sine1_1, nodeColor=<rgba 1 0 0.9609 1>}
  n45 = Sine(None=Panner(Coordinate=TextureCoordinate() {uTiling=4, vTiling=4}) {speedX=-0.25, speedY=-0.25})
  n46 = NamedRerouteDeclaration(None=ConstantBiasScale(None=ComponentMask(None=@n45) {g=true})) {name=Sine2_1, nodeColor=<rgba 0 0.7422 1 1>}
  n47 = Panner(Coordinate=TextureCoordinate() {uTiling=8, vTiling=8}) {speedX=0.4, speedY=0.4}
  n48 = ConstantBiasScale(None=ComponentMask(None=Cosine(None=@n47)) {g=true})
  n49 = TextureCoordinate() {uTiling=5, vTiling=5}
  n50 = ComponentMask(None=Sine(None=Panner(Coordinate=@n49) {speedX=-0.4, speedY=-0.4})) {g=true}
  n51 = Add(
   A=Add(A=Add(A=@n44, B=@n46), B=NamedRerouteDeclaration(None=@n48) {name=Sine3_1, nodeColor=<rgba 1 0 0.9375 1>})
   B=NamedRerouteDeclaration(None=ConstantBiasScale(None=@n50)) {name=Sine4_1, nodeColor=<rgba 0 0.0625 1 1>}
  )
  n52 = Add(
   A=LinearInterpolate(B=Multiply(A=@n41, B=@n42.Param2), Alpha=@n42.Param3) {constB=0, desc=RealAudioSwitch}
   B=LinearInterpolate(B=Multiply(A=@n51) {constB=0.2, desc=DropWaveAmount}, Alpha=@n42.Param4) {constB=0, desc=FakeAudioSwitch}
  )
  n53 = StaticBool() {value=true}
  n54 = ScalarParameter() {defaultValue=0.15, parameterName=Param_13}
  n55 = Multiply(
   A=MaterialFunctionCall(Width=@n52, Sharpness=ScalarParameter() {defaultValue=0.9, parameterName=Param_11}, Direction Switch=@n53) {materialFunction=&GeneratedBand}.Result
   B=OneMinus(None=MaterialFunctionCall(Width=@n52, Sharpness=@n54, Direction Switch=@n53) {materialFunction=&GeneratedBand}.Result)
  )
  n56 = ScalarParameter()
  Saturate(None=Divide(A=Subtract(A=@n55, B=@n56), B=ScalarParameter() {defaultValue=3, parameterName=Param_1}))
 ]
}
