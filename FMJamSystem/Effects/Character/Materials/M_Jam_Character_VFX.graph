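# Translucent unlit Niagara mesh-particle material for the Jam character ring VFX, reacting to music through dynamic parameters.
# Emissive combines gradient masks, panning noise and world-position stripes whose spacing depends on IsDrum, tinted by particle color.
# Four audio bands scale panning sine waves for vertical world position offset, with switches for real-audio and fake-audio drop motion.
# View-angle fade softens opacity, and time of day adjusts the gradient.
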Material /FMJamSystem/Effects/Character/Materials/M_Jam_Character_VFX
graph {
 properties {
  bCastRayTracedShadows=false
  bEnableResponsiveAA=true
  bOutputTranslucentVelocity=true
  bUsedWithNiagaraMeshParticles=true
  blendMode=BLEND_Translucent
  pixelDepthOffsetMode=PDOM_Legacy
  previewMesh=/FMJamSystem/Effects/Character/Meshes/SM_Jam_Character_VFX_Ring
  shadingModel=MSM_Unlit
  usedShadingModels=Unlit
 }
 nodes [
  n1 = NamedRerouteUsage()
  n2 = Add(A=NamedRerouteUsage(), B=NamedRerouteUsage())
  n3 = ParticleColor()
  MaterialFunctionCall(Emissive=Multiply(A=Add(A=@n1, B=@n2), B=@n3.RGB)) {materialFunction=&MF_FX_EmissiveAdjustments}
  Multiply(A=Time(), B=Constant() {r=7})
  n4 = NamedRerouteUsage()
  Multiply(
   A=Multiply(A=@n3.A, B=Multiply(A=@n4) {constB=0.5})
   B=MaterialFunctionCall("FadeSensitivity (optional)"=Constant() {desc=Was previously 2.5, r=4}) {materialFunction=&FadeBasedOnViewAngle}.Result
  )
  n5 = MaterialFunctionCall(Value_Morning=Constant() {r=0.6}, Value_Day=Constant() {r=0.5}, Value_Evening=Constant() {r=0.75}, Value_Night=Constant() {r=0.65}) {materialFunction=&Scalar_Time_of_Day}
  n6 = Constant() {r=1.5}
  n7 = TextureCoordinate()
  n8 = Sine(None=Multiply(A=ComponentMask(None=@n7) {g=true}) {constB=0.5})
  n9 = Power(Base=OneMinus(None=@n8), Exp=Constant() {r=2})
  n10 = Step(Y=@n8, X=Constant() {r=0.15}) {constX=0.5}
  NamedRerouteDeclaration(None=Add(A=LinearInterpolate(A=@n5.Result, B=@n6, Alpha=@n9), B=@n10)) {name=Gradient and Masks, nodeColor=<rgba 0 0 0 1>}
  n11 = Time()
  n12 = Divide(
   A=TransformPosition(None=WorldPosition().XYZ) {transformSourceType=TRANSFORMPOSSOURCE_World}
   B=LinearInterpolate(Alpha=DynamicParameter() {paramNames [IsDrum, LoopAmplitudeMulitiply, DropSwitchRealAudio, DropSwitchFakeAudio]}.IsDrum) {constA=3, constB=4, desc=Stripe Count}
  )
  n13 = MaterialFunctionCall!(
   Width=Constant() {r=0.2}
   Sharpness
   Offset=Frac(None=Multiply(A=@n11) {constB=0, desc=PanSpeed})
   Direction Switch=StaticBool()
   Compare
   Input Coordinates=ComponentMask(None=Add(A=@n12) {constB=0.5}) {b=true}
  ) {materialFunction=/Engine/Functions/Engine_MaterialFunctions03/Procedurals/GeneratedBand}
  n14 = ComponentMask(None=TextureCoordinate()) {g=true}
  n15 = Power(Base=Sine(None=Multiply(A=@n14) {constB=0.5})) {constExponent=5}
  NamedRerouteDeclaration(None=Multiply(A=Multiply(A=@n13.Result, B=@n15))) {name=Stripes, nodeColor=<rgba 0 0 0 1>}
  n16 = DynamicParameter() {desc=Audio values, parameterIndex=1, defaultValue=<rgba 0 0 0 0>, paramNames [Band1,Band2,Band3,Band4]}
  NamedRerouteDeclaration(None=@n16.Band2) {name=Band2, nodeColor=<rgba 1 0 0.6328 1>}
  NamedRerouteDeclaration(None=@n16.Band3) {name=Band3, nodeColor=<rgba 0.2656 0 1 1>}
  NamedRerouteDeclaration(None=@n16.Band4) {name=Band4, nodeColor=<rgba 1 0 0.9609 1>}
  n17 = NamedRerouteUsage()
  n18 = NamedRerouteUsage()
  n19 = NamedRerouteUsage()
  n20 = NamedRerouteUsage()
  n21 = Add(A=Add(A=Multiply(A=@n17, B=@n18), B=Multiply(A=@n19, B=@n20)), B=Multiply(A=NamedRerouteUsage(), B=NamedRerouteUsage()))
  n22 = Multiply(A=NamedRerouteUsage(), B=NamedRerouteUsage())
  n23 = DynamicParameter() {paramNames [IsDrum, LoopAmplitudeMulitiply, DropSwitchRealAudio, DropSwitchFakeAudio]}
  n24 = Multiply(A=Multiply(A=Add(A=@n21, B=@n22)) {constB=75, desc=SinewaveMultiply 500}, B=@n23.LoopAmplitudeMulitiply)
  n25 = TextureCoordinate() {uTiling=3}
  n26 = ComponentMask(None=Sine(None=Panner(Coordinate=@n25) {speedX=0.25})) {r=true}
  n27 = TextureCoordinate() {uTiling=4}
  n28 = ComponentMask(None=Sine(None=Panner(Coordinate=@n27) {speedX=-0.25})) {r=true}
  n29 = Add(
   A=NamedRerouteDeclaration(None=ConstantBiasScale(None=@n26)) {name=Sine1, nodeColor=<rgba 1 0 0.9609 1>}
   B=NamedRerouteDeclaration(None=ConstantBiasScale(None=@n28)) {name=Sine2, nodeColor=<rgba 0 0.7422 1 1>}
  )
  n30 = Cosine(None=Panner(Coordinate=TextureCoordinate() {uTiling=8}) {speedX=0.4})
  n31 = NamedRerouteDeclaration(None=ConstantBiasScale(None=ComponentMask(None=@n30) {r=true})) {name=Sine3, nodeColor=<rgba 1 0 0.9375 1>}
  n32 = Panner(Coordinate=TextureCoordinate() {uTiling=5}) {speedX=-0.4}
  n33 = ConstantBiasScale(None=ComponentMask(None=Sine(None=@n32)) {r=true})
  n34 = Multiply(A=Add(A=Add(A=@n29, B=@n31), B=NamedRerouteDeclaration(None=@n33) {name=Sine4, nodeColor=<rgba 0 0.0625 1 1>})) {constB=7.5, desc=DropWaveAmount}
  NamedRerouteDeclaration(None=Add(
    A=LinearInterpolate(B=@n24, Alpha=@n23.DropSwitchRealAudio) {constB=0, desc=RealAudioSwitch}
    B=LinearInterpolate(B=@n34, Alpha=@n23.DropSwitchFakeAudio) {constB=0, desc=FakeAudioSwitch}
   )) {name=4BandSineWaves, nodeColor=<rgba 0 0 0 1>}
  n35 = Multiply(A=TextureCoordinate(), B=ComponentMask(None=VectorParameter() {defaultValue=<rgba 1 1 0 1>}.RGB) {g=true, r=true})
  n36 = TextureCoordinate() {uTiling=10}
  n37 = Add(
   A=TextureSample(UVs=Panner(Coordinate=@n35) {speedX=-1}) {samplerType=SAMPLERTYPE_Grayscale, texture=/FortWater/Textures/LowResBlurredNoise}.R
   B=TextureSample(UVs=Panner(Coordinate=@n36) {speedX=0.5}) {samplerType=SAMPLERTYPE_Grayscale, texture=&LowResBlurredNoise}.R
  )
  n38 = Constant() {r=3}
  NamedRerouteDeclaration(None=Multiply(A=Multiply(A=@n10) {constB=2}, B=Multiply(A=@n37, B=@n38) {constB=0.25, desc=was 3 before the Param was placed})) {name=Top Band Noise, nodeColor=<rgba 0 0 0 1>}
  NamedRerouteDeclaration(None=@n16.Band1) {name=Band1, nodeColor=<rgba 1 0.7031 0 1>}
  n39 = NamedRerouteUsage()
  n40 = ComponentMask(None=TextureCoordinate()) {g=true}
  n41 = LinearInterpolate(
   A=Constant3Vector() {desc=WorldPositionDirection, constant=<rgba 0 0 1 0>}
   B=Constant3Vector() {desc=WorldPositionDirection, constant=<rgba 0 0 -1 0>}
   Alpha=Step(Y=OneMinus(None=@n40)) {constX=0.5}
  )
  n42 = Constant3Vector() {desc=WorldPositionDirection, constant=<rgba 0 0 -10.5 0>}
  n43 = Constant3Vector() {desc=WorldPositionDirection, constant=<rgba 0 0 10.5 0>}
  n44 = TextureCoordinate()
  n45 = Step(Y=OneMinus(None=ComponentMask(None=@n44) {g=true})) {constX=0.5}
  NamedRerouteDeclaration(None=Add(A=Multiply(A=@n39, B=@n41), B=LinearInterpolate(A=@n42, B=@n43, Alpha=@n45))) {name=Final WPO, nodeColor=<rgba 0 0 0 1>}
  NamedRerouteUsage()
 ]
}
