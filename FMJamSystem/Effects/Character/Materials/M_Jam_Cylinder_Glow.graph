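# Unlit, two-sided translucent glow material for Niagara effects in the Jam system's character effects, named for use on a cylinder.
# Particle color RGB passes through the shared FX emissive adjustments function.
# Particle alpha is scaled by a dimmed, power-sharpened vertical UV gradient, which the FlipGradient parameter can invert.
# A view-angle fade multiplies that alpha to soften the glow by viewing direction.
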Material /FMJamSystem/Effects/Character/Materials/M_Jam_Cylinder_Glow
graph {
 properties {
  bAutomaticallySetUsageInEditor=false
  bUsedWithNiagaraMeshParticles=true
  bUsedWithNiagaraRibbons=true
  bUsedWithNiagaraSprites=true
  blendMode=BLEND_Translucent
  shadingModel=MSM_Unlit
  twoSided=true
  usedShadingModels=Unlit
 }
 nodes [
  n1 = ParticleColor()
  MaterialFunctionCall(Emissive=@n1.RGB) {materialFunction=&MF_FX_EmissiveAdjustments}
  n2 = MaterialFunctionCall() {materialFunction=&LinearGradient}
  n3 = Power(
   Base=LinearInterpolate(A=OneMinus(None=@n2.VGradient), B=@n2.VGradient, Alpha=ScalarParameter() {parameterName=FlipGradient})
   Exp=Constant() {r=2.5}
  )
  n4 = Constant() {r=0.125}
  Multiply(
   A=Multiply(A=@n1.A, B=Multiply(A=@n3, B=@n4))
   B=MaterialFunctionCall("FadeSensitivity (optional)"=Constant() {r=1}) {materialFunction=&FadeBasedOnViewAngle}.Result
  )
 ]
}
